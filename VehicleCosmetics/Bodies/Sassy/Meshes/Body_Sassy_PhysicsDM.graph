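# Physics asset for the Sassy vehicle body cosmetic, providing a simplified collision shape for its mesh.
# A single box on the root bone, offset upward and slightly forward, approximates the whole car body.
# It has no other bodies and no constraints, so it defines no jointed or separately simulated parts.
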PhysicsAsset /VehicleCosmetics/Bodies/Sassy/Meshes/Body_Sassy_PhysicsDM
bodies {root [{shapeType=Box, extentX=444, extentY=216, extentZ=116, shapeName=root_box, center=<11.5 0 91>}]}
constraints []
